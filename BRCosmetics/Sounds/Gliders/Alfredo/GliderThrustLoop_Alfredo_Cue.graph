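# Looping thrust sound cue for the Alfredo glider cosmetic.
# A mixer blends two looping waves, a flight loop and a wind loop, with the flight layer weighted louder.
# Each layer passes through a continuous modulator driven by the AnyDirection parameter, raising pitch and shifting volume within its own range.
# Glider-specific attenuation, sound class and audio properties sheet govern playback.
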SoundCue /BRCosmetics/Sounds/Gliders/Alfredo/GliderThrustLoop_Alfredo_Cue
root=Mixer_0
nodes [
 {id=Mixer_0, children [ModulatorContinuous_0, ModulatorContinuous_1], properties {inputVolume [1.3,0.8]}}
 {
  id=ModulatorContinuous_0
  children [WavePlayer_0]
  properties {
   pitchModulationParams {default=0, maxOutput=1.2, minOutput=1, parameterName=AnyDirection}
   volumeModulationParams {default=0, minOutput=0.5, parameterName=AnyDirection}
  }
 }
 {
  id=ModulatorContinuous_1
  children [WavePlayer_1]
  properties {
   pitchModulationParams {default=0, maxOutput=1.2, minOutput=1, parameterName=AnyDirection}
   volumeModulationParams {default=0, maxOutput=0.7, minOutput=0.2, parameterName=AnyDirection}
  }
 }
 {id=WavePlayer_0, properties {bLooping=true, soundWaveAssetPtr=&/Glider_Alfredo_Flight_LP}}
 {id=WavePlayer_1, properties {bLooping=true, soundWaveAssetPtr=&/Glider_Alfredo_Flight_Wind_LP}}
]
properties {
 attenuationSettings=&Fort_Player_Glider_Thrust_Athena_Att
 audioPropertiesSheet=&APS_Cosmetic_Glider_Loop
 maxDistance=12500
 soundClassObject=&Fort_Gliders_Land
 volumeMultiplier=0.9
 concurrencySet [null]
}
